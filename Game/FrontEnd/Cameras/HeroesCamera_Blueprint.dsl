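; Front-end camera Blueprint for the heroes screen that can spin the hero pawn with mouse movement.
; Activation triggers the camera-switch effect on every TVPostProcessBP actor and enables input; deactivation disables input.
; Tick yaws the cached pawn by mouse delta while MouseDown is set.
; HandleMousePress caches hero pawn 0; HandleMouseRelease clears it but leaves MouseDown true, and both mouse button events are empty.
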(event Camera|Fortnite|EventOnActivated
  (params (PlayerController :type Fort%20Player%20Controller%20Object%20Reference))
  ;[0.5688 1 0] Camera switch post process anim
  (bind _outactors (Actor|GetAllActorsOfClass "/Game/UI/Foundation/Environment/Cameras/TVPostProcessBP.TVPostProcessBP_C"))
  (for _array_element _outactors
    (Class|TVPostProcessBP|FrontEndCameraSwitch _array_element))
  ;[1 1 1] Enable input
  (Input|EnableInput (Game|GetPlayerController 0)))

(event Camera|Fortnite|EventOnDeactivated
  (params (PlayerController :type Fort%20Player%20Controller%20Object%20Reference))
  ;[1 1 1] Disable input
  (Input|DisableInput (Game|GetPlayerController 0)))

(event LeftMouseButton
  (params (Key :type Key%20Structure)))

(event RightMouseButton
  (params (Key :type Key%20Structure)))

(event EventTick
  (params (DeltaSeconds :type Float%20%28single-precision%29))
  (bind _cached_pawn (Variables|Default|GetCached%20Pawn))
  (if (Variables|Default|GetMouseDown)
    (Utilities|IsValid _cached_pawn
      (:Is%20Valid
        (Transformation|SetActorRotation _cached_pawn (Math|Rotator|MakeRotator (.roll (Transformation|GetActorRotation _cached_pawn)) (.pitch (Transformation|GetActorRotation _cached_pawn)) (- (.yaw (Transformation|GetActorRotation _cached_pawn)) (* (Game|Player|GetInputMouseDelta (Game|GetPlayerController 0)) 10))) false)))))

(fn HandleMousePress
  (Variables|Default|SetMouseDown true)
  (bind _returnvalue (FrontEndContext|GetHeroPlayerPawnbyIndex (BlueprintContext|GetFortFrontEndContext) 0))
  (Utilities|IsValid _returnvalue
    (:Is%20Valid
      (Variables|Default|SetCached%20Pawn _returnvalue))))

(fn HandleMouseRelease
  (Variables|Default|SetMouseDown true)
  (Variables|Default|SetCached%20Pawn 0))

(fn ConstructionScript)
