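; Front-end camera Blueprint for the Special Events screen in the Fortnite lobby, reacting to camera activation and deactivation.
; On activation it enables lobby set lighting on every BP_FortniteLobbyLightSwitcher, then triggers the FrontEndCameraSwitch post-process animation on every TVPostProcessBP actor.
; On deactivation it disables that lobby set lighting again.
; Target actors are found by class lookup each time rather than held as references.
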(event Camera|Fortnite|EventOnActivated
  (params (PlayerController :type Fort%20Player%20Controller%20Object%20Reference))
  ;[1 1 1] Turn on necessary lights
  (Utilities|FlowControl|Sequence
    (:then_0
      (bind _outactors (Actor|GetAllActorsOfClass "/Game/Environments/FrontEnd/Blueprints/BP_FortniteLobbyLightSwitcher.BP_FortniteLobbyLightSwitcher_C"))
      (for _array_element _outactors
        (Class|BPFortniteLobbyLightSwitcher|EnableLobbySetLighting _array_element)))
    (:then_1
      ;[0.5688 1 0] Camera switch post process anim
      (bind _outactors_2 (Actor|GetAllActorsOfClass "/Game/UI/Foundation/Environment/Cameras/TVPostProcessBP.TVPostProcessBP_C"))
      (for _array_element_2 _outactors_2
        (Class|TVPostProcessBP|FrontEndCameraSwitch _array_element_2)))))

(event Camera|Fortnite|EventOnDeactivated
  (params (PlayerController :type Fort%20Player%20Controller%20Object%20Reference))
  ;[1 1 1] Turn off necessary lights
  (bind _outactors (Actor|GetAllActorsOfClass "/Game/Environments/FrontEnd/Blueprints/BP_FortniteLobbyLightSwitcher.BP_FortniteLobbyLightSwitcher_C"))
  (for _array_element _outactors
    (Class|BPFortniteLobbyLightSwitcher|DisableLobbySetLighting _array_element)))

(fn ConstructionScript)
