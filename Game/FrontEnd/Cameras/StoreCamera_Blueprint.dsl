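; Front-end store camera that stages the pinata and card reveal sequence.
; On activation it resets time dilation, sets background darkening, fires the TV post-process switch, then finds and initiates the pinata and card reveal actors.
; A reversible timeline lerps between construction-cached start and ground-placeholder transforms, widening FOV, then reports completion to the pinata by direction flag.
; A second reversible timeline widens FOV further for choice packs.
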(event EventBeginPlay)

(event Collision|EventActorBeginOverlap
  (params (OtherActor :type Actor%20Object%20Reference)))

(event EventTick
  (params (DeltaSeconds :type Float%20%28single-precision%29)))

(event Custom|DarkenBGVisibility
  (params (Enable :type Boolean))
  ;[1 1 1] Darken Background Visibility
  (Rendering|SetVisibility (Variables|Default|GetMesh-DarkenBG) Enable true))

(event Camera|Fortnite|EventOnActivated
  (params (PlayerController :type Fort%20Player%20Controller%20Object%20Reference))
  (bind _pinatainlevel (Variables|Default|GetPinataInLevel))
  (bind _cardrevealinlevel (Variables|Default|GetCardRevealInLevel))
  (if (not (Class|BlueprintLibrarySTW|IsinHestiaBeauty))
    ;[1 1 1] DarkenBG control
    ;;[1 1 1] 1=opaque
    (Rendering|Material|SetScalarParameterValue "/Game/FrontEnd/Store/Pinata/Assets/Shared/MPC_Pinata.MPC_Pinata" "BGChange" 1 :!class /Script/Engine.KismetMaterialLibrary)
    (Rendering|SetVisibility (Variables|Default|GetMesh-DarkenBG) true false))
  ;[1 1 1] Ensure Time Dilation
  (Utilities|Time|SetGlobalTimeDilation 1)
  ;[0.5688 1 0] Camera switch post process anim
  (bind _outactors (Actor|GetAllActorsOfClass "/Game/UI/Foundation/Environment/Cameras/TVPostProcessBP.TVPostProcessBP_C"))
  (for _array_element _outactors
    (Class|TVPostProcessBP|FrontEndCameraSwitch _array_element))
  ;[1 1 1] Establish connection to associated actors
  (bind _outactors_2 (Actor|GetAllActorsOfClass "/Game/FrontEnd/Store/Pinata/Blueprints/StorePinataMaster_Parent.StorePinataMaster_Parent_C"))
  (for _array_element_2 _outactors_2
    (Variables|Default|SetPinataInLevel _array_element_2))
  (bind _outactors_3 (Actor|GetAllActorsOfClass "/Game/FrontEnd/Store/Cards/Blueprints/StoreCardReveal_Parent.StoreCardReveal_Parent_C"))
  (for _array_element_3 _outactors_3
    (Variables|Default|SetCardRevealInLevel _array_element_3))
  ;[1 1 1] Initiate associated BPs
  (Utilities|IsValid _pinatainlevel
    (:Is%20Valid
      (Class|StorePinataMasterParent|InitiatePinata _pinatainlevel))
    (:Is%20Not%20Valid
      _))
  (Utilities|IsValid _cardrevealinlevel
    (:Is%20Valid
      (Class|StoreCardRevealParent|InitiatePinata _cardrevealinlevel))))

(event Camera|Fortnite|EventOnDeactivated
  (params (PlayerController :type Fort%20Player%20Controller%20Object%20Reference))
  ;[1 1 1] DarkenBG control
  ;;[1 1 1] 1=opaque
  (Rendering|Material|SetScalarParameterValue "/Game/FrontEnd/Store/Pinata/Assets/Shared/MPC_Pinata.MPC_Pinata" "BGChange" 1 :!class /Script/Engine.KismetMaterialLibrary)
  (Rendering|SetVisibility (Variables|Default|GetMesh-DarkenBG) false false)
  ;[1 1 1] Ensure Time Dilation
  (Utilities|Time|SetGlobalTimeDilation 1))

(event Custom|OpeningCameraTransition
  ;[1 1 1] Camera transition to ground sequence (FOV during pinata drop=75, FOV during cards reveal=90)
  (Variables|Default|SetCameraInStartPos false)
  (bind (_direction _newtrack_0) (Timeline|Timeline_0 PlayFromStart)
    (:Update
      (Transformation|SetActorLocationAndRotation (Math|Vector|Lerp(Vector) (Variables|Default|GetCameraOriginalLoc) (Variables|Default|GetCameraGroundLoc) _newtrack_0) (Math|Rotator|Lerp(Rotator) (Variables|Default|GetCameraOriginalRot) (Variables|Default|GetCameraGroundRot) _newtrack_0))
      (Variables|/Script/Engine.CameraComponent|SetFieldOfView (Math|Float|Lerp 75 90 _newtrack_0) (Variables|CameraActor|GetCameraComponent)))
    (:Finished
      (Utilities|IsValid (Variables|Default|GetPinataInLevel)
        (:Is%20Valid
          (if (Variables|Default|GetCameraInStartPos)
            ;[1 1 1] CameraResetComplete
            (Class|StorePinataMasterParent|CameraResetComplete (Variables|Default|GetPinataInLevel))
            (else
              ;[1 1 1] CameraOpeningTransition: complete
              (Class|StorePinataMasterParent|CameraTransitionComplete (Variables|Default|GetPinataInLevel)))))))))

(event Custom|ResetCamera
  ;[1 1 1] Camera transition to ground sequence (FOV during pinata drop=75, FOV during cards reveal=90)
  (Variables|Default|SetCameraInStartPos true)
  (Timeline|Timeline_0 ReverseFromEnd))

(event Custom|ChoicePackFOV-Out
  ;[1 1 1] Choice Pack FOV
  (bind (_direction _newtrack_0) (Timeline|ChoicePack PlayFromStart)
    (:Update
      (Variables|/Script/Engine.CameraComponent|SetFieldOfView (Math|Float|Lerp 90 110 _newtrack_0) (Variables|CameraActor|GetCameraComponent)))))

(event Custom|ChoicePackFOV-In
  ;[1 1 1] Choice Pack FOV
  (Timeline|ChoicePack ReverseFromEnd))

(fn ConstructionScript
  (bind _cameracomponent (Variables|CameraActor|GetCameraComponent))
  (bind _cameraplaceholderground (Variables|Default|GetCameraPlaceholderGround))
  (Variables|Default|SetCameraOriginalLoc (Transformation|GetWorldLocation _cameracomponent))
  (Variables|Default|SetCameraOriginalRot (Transformation|GetWorldRotation _cameracomponent))
  (Variables|Default|SetCameraGroundLoc (Transformation|GetWorldLocation _cameraplaceholderground))
  (Variables|Default|SetCameraGroundRot (Transformation|GetWorldRotation _cameraplaceholderground)))
